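# EQS query that picks a point for a stuck terrestrial NPC to move to when trying to get free.
# It scores a ring of points around the querier, favoring downhill, on-navmesh, minimum-distance and roughly forward points.
# A projection onto any traversable nav avoids points under water, and the first option also requires pathfinding.
# The fallback option projects its ring onto geometry and drops the pathfinding test.
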EnvQuery /Game/Athena/AI/EQS/Behaviors/EQS_NPC_AttemptToGetUnstuck_Terrestrial
options [
 {
  generator {
   name=EnvQueryGenerator_OnCircle_0
   class=EnvQueryGenerator_OnCircle
   properties {
    circleRadius=500
    numberOfPoints=8
    projectionData {extentX=0, navigationFilter=null, postProjectionVerticalOffset=30, projectDown=1024}
    traceData {navigationFilter=null}
   }
  }
  tests [
   EnvQueryTest_Distance_0 {
    class=EnvQueryTest_Distance
    properties {
     distanceTo=&EnvQueryContext_Querier
     filterType=Range
     floatValueMin=0
     scoringEquation=Linear
     scoringFactor=1
     testComment="Prefer things downhill (for step height reasons)"
     testMode=DistanceZ
     testPurpose=Score
    }
   }
   FortQueryTest_ProjectOnNavMesh_0 {
    class=FortQueryTest_ProjectOnNavMesh
    properties {
     bDefineReferenceValue=false
     boolValue=true
     clampMaxType=None
     clampMinType=None
     filterType=Match
     floatValueMax=0
     floatValueMin=0
     multipleContextFilterOp=AllPass
     multipleContextScoreOp=AverageScore
     normalizationType=Absolute
     referenceValue=0
     scoreClampMax=0
     scoreClampMin=0
     scoringEquation=Constant
     scoringFactor=0.9
     testComment="Prefer points on nav, but care about Z more (for step height)"
     testPurpose=Score
    }
   }
   EnvQueryTest_Distance_1 {
    class=EnvQueryTest_Distance
    properties {
     clampMaxType=SpecifiedValue
     clampMinType=SpecifiedValue
     distanceTo=&EnvQueryContext_Querier
     filterType=Range
     floatValueMin=0
     scoreClampMax=201
     scoreClampMin=200
     scoringEquation=Linear
     scoringFactor=0.75
     testComment=give big score bump to things a min distance away from you
     testMode=Distance3D
     testPurpose=Score
    }
   }
   EnvQueryTest_Dot_0 {
    class=EnvQueryTest_Dot
    properties {
     clampMaxType=SpecifiedValue
     clampMinType=None
     scoreClampMax=0.3
     scoreClampMin=0
     scoringFactor=0.5
     testComment="prefer things in front of self (but only very generally)"
    }
   }
   EnvQueryTest_Project_0 {
    class=EnvQueryTest_Project
    properties {
     scoringFactor=0
     testComment=Project to any traversable so we don't penetrate under water.
     testPurpose=Score
     projectionData {
      navigationFilter=&NavFilter_NPC_AnyTraversable_C
      postProjectionVerticalOffset=30
      projectDown=1000
      projectUp=1000
      traceMode=Navigation
     }
    }
   }
   EnvQueryTest_Pathfinding_0 {class=EnvQueryTest_Pathfinding, properties {filterClass=null, testPurpose=FilterAndScore}}
  ]
 }
 {
  generator {
   name=EnvQueryGenerator_OnCircle_1
   class=EnvQueryGenerator_OnCircle
   properties {
    circleRadius=500
    numberOfPoints=8
    projectionData {extentX=0, navigationFilter=null, postProjectionVerticalOffset=30, projectDown=1024, traceMode=GeometryByChannel}
    traceData {navigationFilter=null}
   }
  }
  tests [
   EnvQueryTest_Distance_2 {
    class=EnvQueryTest_Distance
    properties {
     distanceTo=&EnvQueryContext_Querier
     filterType=Range
     floatValueMin=0
     scoringEquation=Linear
     scoringFactor=1
     testComment="Prefer things downhill (for step height reasons)"
     testMode=DistanceZ
     testPurpose=Score
    }
   }
   FortQueryTest_ProjectOnNavMesh_1 {
    class=FortQueryTest_ProjectOnNavMesh
    properties {
     bDefineReferenceValue=false
     boolValue=true
     clampMaxType=None
     clampMinType=None
     filterType=Match
     floatValueMax=0
     floatValueMin=0
     multipleContextFilterOp=AllPass
     multipleContextScoreOp=AverageScore
     normalizationType=Absolute
     referenceValue=0
     scoreClampMax=0
     scoreClampMin=0
     scoringEquation=Constant
     scoringFactor=0.9
     testComment="Prefer points on nav, but care about Z more (for step height)"
     testPurpose=Score
    }
   }
   EnvQueryTest_Distance_3 {
    class=EnvQueryTest_Distance
    properties {
     clampMaxType=SpecifiedValue
     clampMinType=SpecifiedValue
     distanceTo=&EnvQueryContext_Querier
     filterType=Range
     floatValueMin=0
     scoreClampMax=201
     scoreClampMin=200
     scoringEquation=Linear
     scoringFactor=0.75
     testComment=give big score bump to things a min distance away from you
     testMode=Distance3D
     testPurpose=Score
    }
   }
   EnvQueryTest_Dot_1 {
    class=EnvQueryTest_Dot
    properties {
     clampMaxType=SpecifiedValue
     clampMinType=None
     scoreClampMax=0.3
     scoreClampMin=0
     scoringFactor=0.5
     testComment="prefer things in front of self (but only very generally)"
    }
   }
   EnvQueryTest_Project_1 {
    class=EnvQueryTest_Project
    properties {
     scoringFactor=0
     testComment=Project to any traversable so we don't penetrate under water.
     testPurpose=Score
     projectionData {
      navigationFilter=&NavFilter_NPC_AnyTraversable_C
      postProjectionVerticalOffset=30
      projectDown=1024
      projectUp=1024
      traceMode=Navigation
     }
    }
   }
  ]
 }
]
