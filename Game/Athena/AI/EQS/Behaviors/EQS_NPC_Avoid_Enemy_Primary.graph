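# EQS query that picks a reachable spot for an NPC to avoid its enemy (the Goal).
# Samples a randomly oriented donut arc around the querier, sized by ability BehaviorDistance and projected to navmesh avoiding water.
# Scores favor points far from and lateral to the Goal, away from self, nearby AI and their destinations, plus leader line of sight with an STW-faction tag query.
# The near-identical fallback option drops the LOS raycast.
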EnvQuery /Game/Athena/AI/EQS/Behaviors/EQS_NPC_Avoid_Enemy_Primary
options [
 {
  generator {
   name=EnvQueryGenerator_Donut_0
   class=EnvQueryGenerator_Donut
   properties {
    bDefineArc=true
    bUseSpiralPattern=false
    center=&EnvQueryContext_Querier
    innerRadius=200
    numberOfRings=4
    pointsPerRing=5
    arcDirection {lineTo=/Game/Athena/AI/EQS/Contexts/EnvQueryContext_Querier_RandomPointOnUnitCircle_2D.EnvQueryContext_Querier_RandomPointOnUnitCircle_2D_C}
    outerRadius {
     dataBinding=&self:EnvQueryGenerator_Donut_0.FortAIDataProvider_Ability_0
     dataField=BehaviorDistance
     defaultValue=1500
    }
    projectionData {
     extentX=10
     navigationFilter=&NavFilter_NPC_AvoidAllWater_C
     postProjectionVerticalOffset=30
    }
   }
  }
  tests [
   EnvQueryTest_PathfindingBatch_2 {class=EnvQueryTest_PathfindingBatch, properties {filterClass=&RecastFilter_UseDefaultArea}}
   EnvQueryTest_Distance_5 {
    class=EnvQueryTest_Distance
    properties {
     clampMinType=SpecifiedValue
     distanceTo=&EnvQueryContext_Querier
     filterType=Minimum
     floatValueMin=250
     scoreClampMin=250
     scoringEquation=Constant
     scoringFactor=1
     testComment=de-prioritize points we are very close to
     testMode=Distance3D
     testPurpose=Score
    }
   }
   EnvQueryTest_Distance_6 {
    class=EnvQueryTest_Distance
    properties {
     clampMinType=SpecifiedValue
     distanceTo=&FortQueryContext_NearbyAIPawns
     filterType=Minimum
     floatValueMin=250
     scoreClampMin=250
     scoringEquation=Constant
     scoringFactor=1
     testComment=de-prioritize points near other AI
     testMode=Distance3D
     testPurpose=Score
    }
   }
   EnvQueryTest_Distance_7 {
    class=EnvQueryTest_Distance
    properties {
     clampMinType=SpecifiedValue
     distanceTo=&FortQueryContext_NearbyAIPawnsMoveDestinations
     filterType=Minimum
     floatValueMin=250
     scoreClampMin=250
     scoringEquation=Constant
     scoringFactor=1
     testComment=de-prioritize points near other AI's Movement Destinations
     testMode=Distance3D
     testPurpose=Score
    }
   }
   EnvQueryTest_Distance_14 {
    class=EnvQueryTest_Distance
    properties {
     clampMaxType=SpecifiedValue
     distanceTo=&FortQueryContext_Goal
     filterType=Range
     floatValueMin=0
     scoreClampMax=1500
     scoringEquation=Linear
     scoringFactor=3
     testComment=Pick point farthest from GoalActor
     testMode=Distance3D
     testPurpose=Score
    }
   }
   EnvQueryTest_Dot_1 {
    class=EnvQueryTest_Dot
    properties {
     bAbsoluteValue=true
     clampMaxType=SpecifiedValue
     clampMinType=SpecifiedValue
     scoreClampMax=1
     scoreClampMin=-1
     scoringEquation=InverseLinear
     scoringFactor=1
     testComment=Choose points that are somewhat lateral to your GoalActor. Keeps avoiding AI from being an easy shot.
     lineA {dirMode=TwoPoints, lineFrom=&FortQueryContext_Goal, lineTo=&EnvQueryContext_Querier}
    }
   }
   EnvQueryTest_Distance_15 {
    class=EnvQueryTest_Distance
    properties {
     distanceTo=&EnvQueryContext_Querier
     filterType=Range
     floatValueMin=0
     scoringEquation=Linear
     scoringFactor=0.5
     testComment=This simulates the OnCircle Generator when using a grid generator
     testMode=Distance3D
     testPurpose=Score
    }
   }
   FortQueryTest_IsInLeaderLOS_1 {
    class=FortQueryTest_IsInLeaderLOS
    properties {
     bDefineReferenceValue=false
     bRequireLOSRayCast=true
     boolValue=true
     clampMaxType=None
     clampMinType=None
     filterType=Range
     floatValueMax=1
     floatValueMin=0.6
     minDotProduct=0.5
     multipleContextFilterOp=AllPass
     multipleContextScoreOp=AverageScore
     normalizationType=Absolute
     rayCastItemVerticalOffset=50
     rayCastLeaderVerticalOffset=75
     referenceValue=0
     scoreClampMax=0
     scoreClampMin=0
     scoringEquation=Square
     scoringFactor=3
     testComment=""
     testPurpose=Score
     ownerTagQuery {
      autoDescription=" ANY( Athena.Faction.STW )"
      tokenStreamVersion=0
      userDescription=""
      queryTokenStream [0,1,1,1,0]
      tagDictionary [{tagName=Athena.Faction.STW}]
     }
    }
   }
  ]
 }
 {
  generator {
   name=EnvQueryGenerator_Donut_1
   class=EnvQueryGenerator_Donut
   properties {
    bDefineArc=true
    bUseSpiralPattern=false
    center=&EnvQueryContext_Querier
    innerRadius=200
    numberOfRings=4
    pointsPerRing=5
    arcDirection {lineTo=&EnvQueryContext_Querier_RandomPointOnUnitCircle_2D_C}
    outerRadius {
     dataBinding=&self:EnvQueryGenerator_Donut_1.FortAIDataProvider_Ability_0
     dataField=BehaviorDistance
     defaultValue=1500
    }
    projectionData {
     extentX=10
     navigationFilter=&NavFilter_NPC_AvoidAllWater_C
     postProjectionVerticalOffset=30
    }
   }
  }
  tests [
   EnvQueryTest_PathfindingBatch_4 {class=EnvQueryTest_PathfindingBatch, properties {filterClass=&RecastFilter_UseDefaultArea}}
   EnvQueryTest_Distance_16 {
    class=EnvQueryTest_Distance
    properties {
     clampMinType=SpecifiedValue
     distanceTo=&EnvQueryContext_Querier
     filterType=Minimum
     floatValueMin=250
     scoreClampMin=250
     scoringEquation=Constant
     scoringFactor=1
     testComment=de-prioritize points we are very close to
     testMode=Distance3D
     testPurpose=Score
    }
   }
   EnvQueryTest_Distance_17 {
    class=EnvQueryTest_Distance
    properties {
     clampMinType=SpecifiedValue
     distanceTo=&FortQueryContext_NearbyAIPawns
     filterType=Minimum
     floatValueMin=250
     scoreClampMin=250
     scoringEquation=Constant
     scoringFactor=1
     testComment=de-prioritize points near other AI
     testMode=Distance3D
     testPurpose=Score
    }
   }
   EnvQueryTest_Distance_18 {
    class=EnvQueryTest_Distance
    properties {
     clampMinType=SpecifiedValue
     distanceTo=&FortQueryContext_NearbyAIPawnsMoveDestinations
     filterType=Minimum
     floatValueMin=250
     scoreClampMin=250
     scoringEquation=Constant
     scoringFactor=1
     testComment=de-prioritize points near other AI's Movement Destinations
     testMode=Distance3D
     testPurpose=Score
    }
   }
   EnvQueryTest_Distance_19 {
    class=EnvQueryTest_Distance
    properties {
     clampMaxType=SpecifiedValue
     distanceTo=&FortQueryContext_Goal
     filterType=Range
     floatValueMin=0
     scoreClampMax=1500
     scoringEquation=Linear
     scoringFactor=3
     testComment=Pick point farthest from GoalActor
     testMode=Distance3D
     testPurpose=Score
    }
   }
   EnvQueryTest_Dot_8 {
    class=EnvQueryTest_Dot
    properties {
     bAbsoluteValue=true
     clampMaxType=SpecifiedValue
     clampMinType=SpecifiedValue
     scoreClampMax=1
     scoreClampMin=-1
     scoringEquation=InverseLinear
     scoringFactor=1
     testComment=Choose points that are somewhat lateral to your GoalActor. Keeps avoiding AI from being an easy shot.
     lineA {dirMode=TwoPoints, lineFrom=&FortQueryContext_Goal, lineTo=&EnvQueryContext_Querier}
    }
   }
   EnvQueryTest_Distance_21 {
    class=EnvQueryTest_Distance
    properties {
     distanceTo=&EnvQueryContext_Querier
     filterType=Range
     floatValueMin=0
     scoringEquation=Linear
     scoringFactor=0.5
     testComment=This simulates the OnCircle Generator when using a grid generator
     testMode=Distance3D
     testPurpose=Score
    }
   }
   FortQueryTest_IsInLeaderLOS_0 {
    class=FortQueryTest_IsInLeaderLOS
    properties {
     bDefineReferenceValue=false
     bRequireLOSRayCast=false
     boolValue=true
     clampMaxType=None
     clampMinType=None
     filterType=Range
     floatValueMax=1
     floatValueMin=0.6
     minDotProduct=0.6
     multipleContextFilterOp=AllPass
     multipleContextScoreOp=AverageScore
     normalizationType=Absolute
     rayCastItemVerticalOffset=50
     rayCastLeaderVerticalOffset=50
     referenceValue=0
     scoreClampMax=0
     scoreClampMin=0
     scoringEquation=Square
     scoringFactor=3
     testComment=""
     testPurpose=Score
     ownerTagQuery {
      autoDescription=" ANY( Athena.Faction.STW )"
      tokenStreamVersion=0
      userDescription=""
      queryTokenStream [0,1,1,1,0]
      tagDictionary [{tagName=Athena.Faction.STW}]
     }
    }
   }
  ]
 }
]
