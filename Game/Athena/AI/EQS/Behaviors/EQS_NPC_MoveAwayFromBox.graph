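# EQS query that finds a nearby reachable spot for an NPC to move to outside a box, such as when clearing an area.
# The generator takes box center and extents from AI bot extra data contexts, pads it horizontally, and lays rings of navmesh-projected points around it.
# A single path cost test filters by a maximum and scores with negative linear weight, favoring the cheapest path out.
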EnvQuery /Game/Athena/AI/EQS/Behaviors/EQS_NPC_MoveAwayFromBox
options [{
  generator {
   name=FortQueryGenerator_PointsOutsideBox_0
   class=FortQueryGenerator_PointsOutsideBox
   properties {
    bAutoSortTests=true
    bCanRunAsync=false
    envQueryResultNormalizationOption=Default
    extraXExtent=200
    extraYExtent=200
    extraZExtent=0
    generateInBoxCenter=/Script/FortniteAI.FortQueryContext_AIBotExtraData_Box_Center
    generateInBoxExtent=/Script/FortniteAI.FortQueryContext_AIBotExtraData_Box_Extents
    navDataOverrideContext=null
    numberOfRingsAroundVolume=2
    optionName=""
    spaceBetween=150
    projectionData {
     bCanDisableTrace=true
     bCanProjectDown=true
     bCanTraceOnGeometry=true
     bCanTraceOnNavMesh=true
     bOnlyBlockingHits=true
     bTraceComplex=false
     extentX=0
     extentY=10
     extentZ=10
     navigationFilter=null
     postProjectionVerticalOffset=0
     projectDown=1024
     projectUp=1024
     serializedChannel=ECC_Visibility
     traceChannel=TraceTypeQuery1
     traceMode=Navigation
     traceProfileName=None
     traceShape=Line
     versionNum=1
    }
   }
  }
  tests [EnvQueryTest_Pathfinding_10 {
    class=EnvQueryTest_Pathfinding
    properties {
     filterClass=null
     filterType=Maximum
     floatValueMin=0
     scoringEquation=Linear
     scoringFactor=-1
     testMode=PathCost
     testPurpose=FilterAndScore
    }
   }]
 }]
