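# EQS query that picks a nearby spot for an NPC to move away from the actor held in its AI bot extra data.
# Generates a ring of points, projected onto navmesh with a water-avoiding filter.
# A weighted dot test scores points against the actor-to-querier direction, though its comment says it favors perpendicular points.
# A pathfinding test filters by path cost and favors cheaper paths.
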EnvQuery /Game/Athena/AI/EQS/Behaviors/EQS_NPC_MoveAwayFromPawn
options [{
  generator {
   name=EnvQueryGenerator_OnCircle_0
   class=EnvQueryGenerator_OnCircle
   properties {
    circleRadius=400
    numberOfPoints=16
    projectionData {
     extentX=150
     navigationFilter=&NavFilter_NPC_AvoidAllWater_C
     postProjectionVerticalOffset=0
     projectDown=1024
    }
    traceData {navigationFilter=null, traceMode=None}
   }
  }
  tests [
   EnvQueryTest_Dot_1 {
    class=EnvQueryTest_Dot
    properties {
     clampMaxType=None
     clampMinType=None
     floatValueMax=100
     floatValueMin=70
     scoreClampMax=0
     scoreClampMin=0
     scoringFactor=3
     testComment=Favor points perpendicular
     lineA {dirMode=TwoPoints, lineFrom=/Script/FortniteAI.FortQueryContext_AIBotExtraData_Actor, lineTo=&EnvQueryContext_Querier}
    }
   }
   EnvQueryTest_Pathfinding_0 {
    class=EnvQueryTest_Pathfinding
    properties {
     filterClass=null
     filterType=Maximum
     floatValueMin=0
     scoringEquation=Linear
     scoringFactor=-1
     testMode=PathCost
     testPurpose=FilterAndScore
    }
   }
  ]
 }]
