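# EQS query, named as a backup, that picks NPC melee reposition points at minimum preferred range from the goal.
# It generates points on a narrow arc around the goal facing the querier, with radius bound to ability BehaviorDistance and water-avoiding navigation.
# Scoring favors points directly away from the goal and, via a near-step clamp, farther from the querier.
# Points are reprojected onto traversable navmesh, with leader line-of-sight scoring tagged for the STW faction.
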EnvQuery /Game/Athena/AI/EQS/Behaviors/EQS_NPC_Reposition_MinPreferredRange_Melee_Backup
options [{
  generator {
   name=EnvQueryGenerator_OnCircle_0
   class=EnvQueryGenerator_OnCircle
   properties {
    arcAngle=45
    bDefineArc=true
    circleCenter=&FortQueryContext_Goal
    numberOfPoints=7
    arcDirection {lineFrom=&FortQueryContext_Goal, lineTo=&EnvQueryContext_Querier}
    circleRadius {
     dataBinding=&self:EnvQueryGenerator_OnCircle_0.FortAIDataProvider_Ability_0
     dataField=BehaviorDistance
     defaultValue=500
    }
    projectionData {
     extentX=10
     navigationFilter=&NavFilter_NPC_AvoidAllWater_C
     postProjectionVerticalOffset=30
     projectDown=1024
    }
    traceData {navigationFilter=&NavFilter_NPC_AvoidAllWater_C}
   }
  }
  tests [
   EnvQueryTest_Distance_3 {
    class=EnvQueryTest_Distance
    properties {
     distanceTo=&EnvQueryContext_Querier
     filterType=Range
     floatValueMin=0
     scoreClampMax=201
     scoreClampMin=200
     scoringEquation=Linear
     scoringFactor=0.5
     testComment=Prefer farther away
     testMode=Distance3D
     testPurpose=Score
    }
   }
   EnvQueryTest_Dot_1 {
    class=EnvQueryTest_Dot
    properties {
     clampMaxType=SpecifiedValue
     clampMinType=None
     scoreClampMax=0.3
     scoreClampMin=0
     scoringFactor=1
     testComment=prefer points directly away from goal.
     lineA {dirMode=TwoPoints, lineFrom=&FortQueryContext_Goal, lineTo=&EnvQueryContext_Querier}
     lineB {lineFrom=&FortQueryContext_Goal}
    }
   }
   EnvQueryTest_Project_1 {
    class=EnvQueryTest_Project
    properties {
     scoringFactor=0
     testComment=Project to any traversable so we don't penetrate under water.
     testPurpose=Score
     projectionData {
      navigationFilter=&NavFilter_NPC_AnyTraversable_C
      postProjectionVerticalOffset=30
      projectDown=1024
      projectUp=1024
      traceMode=Navigation
     }
    }
   }
   FortQueryTest_IsInLeaderLOS_1 {
    class=FortQueryTest_IsInLeaderLOS
    properties {
     bDefineReferenceValue=false
     bRequireLOSRayCast=false
     boolValue=true
     clampMaxType=None
     clampMinType=None
     filterType=Range
     floatValueMax=1
     floatValueMin=0.6
     minDotProduct=0.6
     multipleContextFilterOp=AllPass
     multipleContextScoreOp=AverageScore
     normalizationType=Absolute
     rayCastItemVerticalOffset=50
     rayCastLeaderVerticalOffset=50
     referenceValue=0
     scoreClampMax=0
     scoreClampMin=0
     scoringEquation=Square
     scoringFactor=1
     testComment=""
     testPurpose=Score
     ownerTagQuery {
      autoDescription=" ANY( Athena.Faction.STW )"
      tokenStreamVersion=0
      userDescription=""
      queryTokenStream [0,1,1,1,0]
      tagDictionary [{tagName=Athena.Faction.STW}]
     }
    }
   }
  ]
 }]
