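# EQS query that picks a spot for an NPC to reposition into melee range of its goal.
# A circle of points around the goal, oriented toward the querier and sized by the ability's behavior distance, is projected onto water-avoiding navmesh.
# Unreachable points and those too near other AI pawns or their move destinations are filtered out.
# Scoring favors spacing from other AI and closeness to the goal.
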EnvQuery /Game/Athena/AI/EQS/Behaviors/EQS_NPC_Reposition_MoveToMelee
options [{
  generator {
   name=EnvQueryGenerator_OnCircle_1
   class=EnvQueryGenerator_OnCircle
   properties {
    circleCenter=&FortQueryContext_Goal
    numberOfPoints=12
    arcDirection {lineFrom=&FortQueryContext_Goal, lineTo=&EnvQueryContext_Querier}
    circleRadius {
     dataBinding=&self:EnvQueryGenerator_OnCircle_1.FortAIDataProvider_Ability_0
     dataField=BehaviorDistance
     defaultValue=500
    }
    projectionData {
     extentX=10
     navigationFilter=&NavFilter_NPC_AvoidAllWater_C
     postProjectionVerticalOffset=30
     projectDown=1024
    }
    traceData {navigationFilter=&NavFilter_NPC_AvoidAllWater_C}
   }
  }
  tests [
   FortQueryTest_PathfindingBatch_0 {
    class=FortQueryTest_PathfindingBatch
    properties {
     bDefineReferenceValue=false
     boolValue=true
     clampMaxType=None
     clampMinType=None
     context=&EnvQueryContext_Querier
     filterClass=&NavFilter_NPC_AvoidAllWater_C
     filterType=Match
     floatValueMax=1000
     floatValueMin=1000
     multipleContextFilterOp=AllPass
     multipleContextScoreOp=AverageScore
     navDataOverrideContext=null
     normalizationType=Absolute
     pathFromContext=true
     referenceValue=0
     scanRangeMultiplier=1.5
     scoreClampMax=0
     scoreClampMin=0
     scoringEquation=Constant
     scoringFactor=1
     skipUnreachable=true
     testComment=Throw out things you cannot path to.
     testMode=PathExist
     testPurpose=Filter
     navFilterTag {tagName=NavFilter.NPC.Reposition}
    }
   }
   EnvQueryTest_Distance_0 {
    class=EnvQueryTest_Distance
    properties {
     distanceTo=&FortQueryContext_NearbyAIPawns
     filterType=Minimum
     floatValueMin=50
     scoreClampMax=201
     scoreClampMin=200
     scoringEquation=Linear
     scoringFactor=0.25
     testComment=Prefer farther away
     testMode=Distance3D
     testPurpose=FilterAndScore
    }
   }
   EnvQueryTest_Distance_1 {
    class=EnvQueryTest_Distance
    properties {
     distanceTo=&FortQueryContext_NearbyAIPawnsMoveDestinations
     filterType=Minimum
     floatValueMin=50
     scoringEquation=Linear
     scoringFactor=0.25
     testComment=Throw out points near other AI's Movement Destinations
     testMode=Distance3D
     testPurpose=FilterAndScore
    }
   }
   EnvQueryTest_Distance_2 {
    class=EnvQueryTest_Distance
    properties {
     distanceTo=&FortQueryContext_Goal
     filterType=Range
     floatValueMin=0
     scoringEquation=InverseLinear
     scoringFactor=1
     testComment=""
     testMode=Distance3D
     testPurpose=Score
    }
   }
   EnvQueryTest_Project_1 {
    class=EnvQueryTest_Project
    properties {
     scoringFactor=0
     testComment=Project to any traversable so we don't penetrate under water.
     testPurpose=Score
     projectionData {
      navigationFilter=&NavFilter_NPC_AnyTraversable_C
      postProjectionVerticalOffset=30
      projectDown=1024
      projectUp=1024
      traceMode=Navigation
     }
    }
   }
  ]
 }]
