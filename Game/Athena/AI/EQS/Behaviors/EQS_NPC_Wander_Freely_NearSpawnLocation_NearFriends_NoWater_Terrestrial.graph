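# EQS query that picks free-wander destinations for terrestrial NPCs, keeping them near their spawn location and friends while avoiding water.
# A donut generator builds an arc along a random direction around the querier, sized by an ability's BehaviorDistance and projected with a water-avoiding nav filter.
# Tests filter out unreachable points and those crowding nearby AI or their move destinations.
# Scoring heavily weights an ability-bound spawn ring, plus nearness to friends, distance from self, facing and slight randomness.
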EnvQuery /Game/Athena/AI/EQS/Behaviors/EQS_NPC_Wander_Freely_NearSpawnLocation_NearFriends_NoWater_Terrestrial
options [{
  generator {
   name=EnvQueryGenerator_Donut_0
   class=EnvQueryGenerator_Donut
   properties {
    bDefineArc=true
    bUseSpiralPattern=false
    center=&EnvQueryContext_Querier
    innerRadius=200
    numberOfRings=4
    pointsPerRing=5
    arcDirection {lineTo=/Game/Athena/AI/EQS/Contexts/EnvQueryContext_Querier_RandomPointOnUnitCircle_2D.EnvQueryContext_Querier_RandomPointOnUnitCircle_2D_C}
    outerRadius {
     dataBinding=&self:EnvQueryGenerator_Donut_0.FortAIDataProvider_Ability_2
     dataField=BehaviorDistance
     defaultValue=1500
    }
    projectionData {
     extentX=10
     navigationFilter=&NavFilter_NPC_AvoidAllWater_C
     postProjectionVerticalOffset=30
    }
   }
  }
  tests [
   EnvQueryTest_PathfindingBatch_3 {
    class=EnvQueryTest_PathfindingBatch
    properties {filterClass=&NavFilter_NPC_ExcludeAllWater_C}
   }
   EnvQueryTest_Distance_17 {
    class=EnvQueryTest_Distance
    properties {
     distanceTo=&EnvQueryContext_Querier
     filterType=Range
     floatValueMin=0
     scoringEquation=Linear
     testComment=Pick point farthest from Self
     testMode=Distance2D
     testPurpose=Score
     scoringFactor {dataBinding=null, dataField=FloatValue, defaultValue=1}
    }
   }
   EnvQueryTest_Dot_3 {
    class=EnvQueryTest_Dot
    properties {
     clampMaxType=None
     clampMinType=None
     scoreClampMax=0
     scoreClampMin=0
     testComment=querier facing item
     scoringFactor {dataBinding=null, dataField=FloatValue, defaultValue=0.2}
    }
   }
   FortQueryTest_Random_3 {class=FortQueryTest_Random, properties {floatValueMax=0.1, scoringFactor {dataBinding=null, dataField=FloatValue, defaultValue=0.1}}}
   EnvQueryTest_Distance_18 {
    class=EnvQueryTest_Distance
    properties {
     clampMaxType=SpecifiedValue
     clampMinType=SpecifiedValue
     distanceTo=/Script/FortniteAIServer.FortQueryContext_AIPawnSpawnLocation
     filterType=Range
     floatValueMin=0
     scoringEquation=InverseLinear
     testComment=Ring around Spawn Location
     testMode=Distance2D
     testPurpose=Score
     scoreClampMax {
      dataBinding=&self:EnvQueryTest_Distance_18.FortAIDataProvider_Ability_1
      dataField=BehaviorDistance
      defaultValue=25000
     }
     scoreClampMin {
      dataBinding=&self:EnvQueryTest_Distance_18.FortAIDataProvider_Ability_0
      dataField=BehaviorDistance
      defaultValue=15000
     }
     scoringFactor {dataBinding=null, dataField=FloatValue, defaultValue=10}
    }
   }
   EnvQueryTest_Distance_19 {
    class=EnvQueryTest_Distance
    properties {
     distanceTo=&FortQueryContext_NearbyAIPawns
     filterType=Minimum
     floatValueMin=250
     scoringEquation=InverseLinear
     scoringFactor=3
     testComment=Try not to overlap AI Pawns
     testMode=Distance3D
     testPurpose=Filter
    }
   }
   EnvQueryTest_Distance_20 {
    class=EnvQueryTest_Distance
    properties {
     distanceTo=/Script/FortniteAIServer.FortQueryContext_NearbyFriendlyAIPawnsAndPlayerPawns
     filterType=Range
     floatValueMin=0
     scoringEquation=InverseLinear
     scoringFactor=3
     testComment=Stay near friends
     testMode=Distance3D
     testPurpose=Score
    }
   }
   EnvQueryTest_Distance_22 {
    class=EnvQueryTest_Distance
    properties {
     distanceTo=&FortQueryContext_NearbyAIPawnsMoveDestinations
     filterType=Minimum
     floatValueMin=250
     scoringEquation=Linear
     scoringFactor=1
     testComment=not near other pawn move destinations
     testMode=Distance3D
     testPurpose=Filter
    }
   }
   EnvQueryTest_Distance_23 {
    class=EnvQueryTest_Distance
    properties {
     distanceTo=&EnvQueryContext_Querier
     filterType=Range
     floatValueMin=0
     scoringEquation=Linear
     scoringFactor=0.5
     testComment=This simulates the OnCircle Generator when using a grid generator
     testMode=Distance3D
     testPurpose=Score
    }
   }
  ]
 }]
