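# EQS query that picks a wander destination for land NPCs, favoring points near their goal.
# Generates donut arcs around the querier in a random direction, with outer radius bound to ability BehaviorDistance.
# Filters unreachable points and those near other AI or their destinations, then scores distance from self, goal proximity, facing and, heavily, dry ground.
# A zero-weight projection keeps points above water; the second option repeats the first's setup.
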EnvQuery /Game/Athena/AI/EQS/Behaviors/EQS_NPC_Wander_NearGoal_Terrestrial
options [
 {
  generator {
   name=EnvQueryGenerator_Donut_0
   class=EnvQueryGenerator_Donut
   properties {
    bDefineArc=true
    bUseSpiralPattern=false
    center=&EnvQueryContext_Querier
    innerRadius=200
    numberOfRings=4
    pointsPerRing=5
    arcDirection {lineTo=/Game/Athena/AI/EQS/Contexts/EnvQueryContext_Querier_RandomPointOnUnitCircle_2D.EnvQueryContext_Querier_RandomPointOnUnitCircle_2D_C}
    outerRadius {
     dataBinding=&self:EnvQueryGenerator_Donut_0.FortAIDataProvider_Ability_0
     dataField=BehaviorDistance
     defaultValue=1500
    }
    projectionData {
     extentX=10
     navigationFilter=&NavFilter_NPC_AvoidAllWater_C
     postProjectionVerticalOffset=30
    }
   }
  }
  tests [
   FortQueryTest_PathfindingBatch_4 {
    class=FortQueryTest_PathfindingBatch
    properties {
     bDefineReferenceValue=false
     boolValue=true
     clampMaxType=None
     clampMinType=None
     context=&EnvQueryContext_Querier
     filterClass=&NavFilter_NPC_AvoidAllWater_C
     filterType=Match
     floatValueMax=1000
     floatValueMin=1000
     multipleContextFilterOp=AllPass
     multipleContextScoreOp=AverageScore
     navDataOverrideContext=null
     normalizationType=Absolute
     pathFromContext=true
     referenceValue=0
     scanRangeMultiplier=1.5
     scoreClampMax=0
     scoreClampMin=0
     scoringEquation=Constant
     scoringFactor=1
     skipUnreachable=true
     testComment=Throw out things you cannot path to.
     testMode=PathExist
     testPurpose=Filter
     navFilterTag {tagName=NavFilter.NPC.Wander}
    }
   }
   EnvQueryTest_Distance_40 {
    class=EnvQueryTest_Distance
    properties {
     distanceTo=&EnvQueryContext_Querier
     filterType=Range
     floatValueMin=0
     scoringEquation=Linear
     scoringFactor=1
     testComment=Pick point farthest from Self
     testMode=Distance3D
     testPurpose=Score
    }
   }
   EnvQueryTest_Dot_19 {
    class=EnvQueryTest_Dot
    properties {
     clampMaxType=SpecifiedValue
     clampMinType=SpecifiedValue
     referenceValue=0.707
     testComment=Prefer in front of Self
     scoreClampMax {dataBinding=null, dataField=FloatValue, defaultValue=0.4}
     scoreClampMin {dataBinding=null, dataField=FloatValue, defaultValue=0}
     scoringFactor {dataBinding=null, dataField=FloatValue, defaultValue=0.25}
    }
   }
   EnvQueryTest_Distance_41 {
    class=EnvQueryTest_Distance
    properties {
     distanceTo=&FortQueryContext_Goal
     filterType=Range
     floatValueMin=0
     scoringEquation=InverseLinear
     scoringFactor=1
     testComment=Prefer points near the goal
     testMode=Distance3D
     testPurpose=Score
    }
   }
   EnvQueryTest_Distance_42 {
    class=EnvQueryTest_Distance
    properties {
     distanceTo=&FortQueryContext_NearbyAIPawns
     filterType=Minimum
     floatValueMin=250
     scoringEquation=Linear
     scoringFactor=1
     testComment=Throw out points near other AI
     testMode=Distance3D
     testPurpose=Filter
    }
   }
   EnvQueryTest_Distance_43 {
    class=EnvQueryTest_Distance
    properties {
     distanceTo=&FortQueryContext_NearbyAIPawnsMoveDestinations
     filterType=Minimum
     floatValueMin=250
     scoringEquation=Linear
     scoringFactor=1
     testComment=Throw out points near other AI's Movement Destinations
     testMode=Distance3D
     testPurpose=Filter
    }
   }
   EnvQueryTest_Project_20 {
    class=EnvQueryTest_Project
    properties {
     scoringFactor=2
     testComment=Give LOTS of score to things NOT on Water. We only use Water if we really need to.
     testPurpose=Score
     projectionData {
      navigationFilter=&NavFilter_NPC_ExcludeAllWater_C
      postProjectionVerticalOffset=30
      projectDown=1024
      projectUp=1024
      traceMode=Navigation
     }
    }
   }
   EnvQueryTest_Project_21 {
    class=EnvQueryTest_Project
    properties {
     scoringFactor=0
     testComment=Project to any traversable so we don't penetrate under water.
     testPurpose=Score
     projectionData {
      navigationFilter=&NavFilter_NPC_AnyTraversable_C
      postProjectionVerticalOffset=30
      projectDown=1024
      projectUp=1024
      traceMode=Navigation
     }
    }
   }
  ]
 }
 {
  generator {
   name=EnvQueryGenerator_Donut_1
   class=EnvQueryGenerator_Donut
   properties {
    bDefineArc=true
    bUseSpiralPattern=false
    center=&EnvQueryContext_Querier
    innerRadius=200
    numberOfRings=4
    pointsPerRing=5
    arcDirection {lineTo=&EnvQueryContext_Querier_RandomPointOnUnitCircle_2D_C}
    outerRadius {
     dataBinding=&self:EnvQueryGenerator_Donut_1.FortAIDataProvider_Ability_0
     dataField=BehaviorDistance
     defaultValue=1500
    }
    projectionData {
     extentX=10
     navigationFilter=&NavFilter_NPC_AvoidAllWater_C
     postProjectionVerticalOffset=30
    }
   }
  }
  tests [
   FortQueryTest_PathfindingBatch_2 {
    class=FortQueryTest_PathfindingBatch
    properties {
     bDefineReferenceValue=false
     boolValue=true
     clampMaxType=None
     clampMinType=None
     context=&EnvQueryContext_Querier
     filterClass=&NavFilter_NPC_AvoidAllWater_C
     filterType=Match
     floatValueMax=1000
     floatValueMin=1000
     multipleContextFilterOp=AllPass
     multipleContextScoreOp=AverageScore
     navDataOverrideContext=null
     normalizationType=Absolute
     pathFromContext=true
     referenceValue=0
     scanRangeMultiplier=1.5
     scoreClampMax=0
     scoreClampMin=0
     scoringEquation=Constant
     scoringFactor=1
     skipUnreachable=true
     testComment=Throw out things you cannot path to.
     testMode=PathExist
     testPurpose=Filter
     navFilterTag {tagName=NavFilter.NPC.Wander}
    }
   }
   EnvQueryTest_Distance_25 {
    class=EnvQueryTest_Distance
    properties {
     distanceTo=&EnvQueryContext_Querier
     filterType=Range
     floatValueMin=0
     scoringEquation=Linear
     scoringFactor=1
     testComment=Pick point farthest from Self
     testMode=Distance3D
     testPurpose=Score
    }
   }
   EnvQueryTest_Dot_12 {
    class=EnvQueryTest_Dot
    properties {
     clampMaxType=SpecifiedValue
     clampMinType=SpecifiedValue
     referenceValue=0.707
     testComment=Prefer in front of Self
     scoreClampMax {dataBinding=null, dataField=FloatValue, defaultValue=0.4}
     scoreClampMin {dataBinding=null, dataField=FloatValue, defaultValue=0}
     scoringFactor {dataBinding=null, dataField=FloatValue, defaultValue=0.25}
    }
   }
   EnvQueryTest_Distance_26 {
    class=EnvQueryTest_Distance
    properties {
     distanceTo=&FortQueryContext_Goal
     filterType=Range
     floatValueMin=0
     scoringEquation=InverseLinear
     scoringFactor=1
     testComment=Prefer points near the goal
     testMode=Distance3D
     testPurpose=Score
    }
   }
   EnvQueryTest_Distance_27 {
    class=EnvQueryTest_Distance
    properties {
     distanceTo=&FortQueryContext_NearbyAIPawns
     filterType=Minimum
     floatValueMin=250
     scoringEquation=Linear
     scoringFactor=1
     testComment=Throw out points near other AI
     testMode=Distance3D
     testPurpose=Filter
    }
   }
   EnvQueryTest_Distance_28 {
    class=EnvQueryTest_Distance
    properties {
     distanceTo=&FortQueryContext_NearbyAIPawnsMoveDestinations
     filterType=Minimum
     floatValueMin=250
     scoringEquation=Linear
     scoringFactor=1
     testComment=Throw out points near other AI's Movement Destinations
     testMode=Distance3D
     testPurpose=Filter
    }
   }
   EnvQueryTest_Project_13 {
    class=EnvQueryTest_Project
    properties {
     scoringFactor=2
     testComment=Give LOTS of score to things NOT on Water. We only use Water if we really need to.
     testPurpose=Score
     projectionData {
      navigationFilter=&NavFilter_NPC_ExcludeAllWater_C
      postProjectionVerticalOffset=30
      projectDown=1024
      projectUp=1024
      traceMode=Navigation
     }
    }
   }
   EnvQueryTest_Project_14 {
    class=EnvQueryTest_Project
    properties {
     scoringFactor=0
     testComment=Project to any traversable so we don't penetrate under water.
     testPurpose=Score
     projectionData {
      navigationFilter=&NavFilter_NPC_AnyTraversable_C
      postProjectionVerticalOffset=30
      projectDown=1024
      projectUp=1024
      traceMode=Navigation
     }
    }
   }
  ]
 }
]
